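# Third-person sound cue for a soldier's hard landing on snow in Athena, routed to the enemy foley sound class with landing attenuation.
# Each branch mixes a random generic soldier landing with a random snow step, then modulates pitch and volume.
# A quality switch gives the first two levels six snow variations and the last two only three.
# Per-owner concurrency stops the oldest sound, alongside a global footstep concurrency set.
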SoundCue /Game/Athena/Sounds/Character/Footsteps/3P/Snow/Athena_Soldier_Snow_LandHard_3P_Cue
root=QualityLevel_0
nodes [
 {id=Mixer_0, children [Random_1,Random_2], properties {inputVolume [0.5,0.7]}}
 {id=Mixer_2, children [Random_3,Random_0], properties {inputVolume [0.5,0.7]}}
 {id=Modulator_0, children [Mixer_0], properties {pitchMax=1.1, pitchMin=0.9, volumeMax=1.3, volumeMin=0.7}}
 {id=Modulator_4, children [Mixer_2], properties {pitchMax=1.1, pitchMin=0.9, volumeMax=1.3, volumeMin=0.7}}
 {id=QualityLevel_0, children [Modulator_4,Modulator_4,Modulator_0,Modulator_0]}
 {
  id=Random_0
  children [WavePlayer_0,WavePlayer_1,WavePlayer_6,WavePlayer_7,WavePlayer_8,WavePlayer_9]
  properties {bShouldExcludeFromBranchCulling=true, weights [1,1,1,1,1,1]}
 }
 {id=Random_1, children [WavePlayer_2,WavePlayer_3], properties {bShouldExcludeFromBranchCulling=true, weights [1,1]}}
 {id=Random_2, children [WavePlayer_4, WavePlayer_5, WavePlayer_10], properties {bShouldExcludeFromBranchCulling=true, weights [1,1,1]}}
 {id=Random_3, children [WavePlayer_19, WavePlayer_20], properties {bShouldExcludeFromBranchCulling=true, weights [1,1]}}
 {id=WavePlayer_0, properties {soundWaveAssetPtr=/Game/Sounds/Fort_Step_Sounds/Soldier/Soldier_Step_Land_Snow_01}}
 {id=WavePlayer_1, properties {soundWaveAssetPtr=/Game/Sounds/Fort_Step_Sounds/Soldier/Soldier_Step_Land_Snow_02}}
 {id=WavePlayer_10, properties {soundWaveAssetPtr=/Game/Sounds/Fort_Step_Sounds/Soldier/Soldier_Step_Land_Snow_03}}
 {id=WavePlayer_19, properties {soundWaveAssetPtr=/Game/Sounds/Fort_Step_Sounds/Human/soldier_land_01}}
 {id=WavePlayer_2, properties {soundWaveAssetPtr=&soldier_land_01}}
 {id=WavePlayer_20, properties {soundWaveAssetPtr=/Game/Sounds/Fort_Step_Sounds/Human/soldier_land_02}}
 {id=WavePlayer_3, properties {soundWaveAssetPtr=&soldier_land_02}}
 {id=WavePlayer_4, properties {soundWaveAssetPtr=&Soldier_Step_Land_Snow_01}}
 {id=WavePlayer_5, properties {soundWaveAssetPtr=&Soldier_Step_Land_Snow_02}}
 {id=WavePlayer_6, properties {soundWaveAssetPtr=&Soldier_Step_Land_Snow_03}}
 {id=WavePlayer_7, properties {soundWaveAssetPtr=/Game/Sounds/Fort_Step_Sounds/Soldier/Soldier_Step_Land_Snow_04}}
 {id=WavePlayer_8, properties {soundWaveAssetPtr=/Game/Sounds/Fort_Step_Sounds/Soldier/Soldier_Step_Land_Snow_05}}
 {id=WavePlayer_9, properties {soundWaveAssetPtr=/Game/Sounds/Fort_Step_Sounds/Soldier/Soldier_Step_Land_Snow_06}}
]
properties {
 attenuationSettings=/Game/Sounds/Attenuation/Fort_Player_Land_3P_Athena_Att
 maxDistance=3584
 soundClassObject=&Fort_Player_Foley_3P_Enemy_Athena
 volumeMultiplier=1.4
 concurrencyOverrides {bLimitToOwner=true, maxCount=4, resolutionRule=StopOldest}
 concurrencySet [&Athena_Footstep_Global_Conc]
}
